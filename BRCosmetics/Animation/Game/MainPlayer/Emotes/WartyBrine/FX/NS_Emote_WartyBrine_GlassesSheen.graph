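# Niagara effect for the WartyBrine emote that, going by its name, renders a brief sheen on glasses.
# A single GPU emitter, named BlowingParticles despite the sheen role, burst-spawns short-lived mesh particles in a local-space offset with time-of-day tinted color.
# A linear curve drives a dynamic material parameter downward over particle life alongside constant parameters, so the mesh's material animates the look rather than particle motion.
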NiagaraSystem /BRCosmetics/Animation/Game/MainPlayer/Emotes/WartyBrine/FX/NS_Emote_WartyBrine_GlassesSheen
dependencies {usedDataInterfaces [&NiagaraDataInterfaceCurve], usedDynamicInputs [], usedModules [&SystemState]}
emitters {BlowingParticles {
  simTarget=GPUComputeSim
  emitterUpdateScript [
   &EmitterState()
   &SpawnBurst_Instantaneous(Use Spawn Probability=false, UseLoopCountLimit=false)
  ]
  particleSpawnScript [&InitializeParticle(
    Color=&FNTimeOfDayColor(
     Night Color=<rgba 0.0815 0.1401 0.8698 1>
     Morning Color=<rgba 0.502 0.3059 0.1843 1>
     Day Color=<rgba 0.5 0.5 0.5 1>
     Evening Color=<rgba 0.502 0.3059 0.1843 1>
    )
    Lifetime=1.2
    Mass Max=1.5
    Mass Mode=::NewEnumerator3
    Position Offset Coordinate Space=::Local
    Sprite Rotation Mode=::NewEnumerator2
    Sprite Size Mode=::NewEnumerator1
    Uniform Sprite Size Max=8
    Uniform Sprite Size Min=4
    UsePositionOffset=true
    Write Material Random=true
   )]
  particleUpdateScript [
   &ParticleState()
   &DynamicMaterialParameters(
    Index 0 Param 1=&FloatFromCurve(FloatCurve=DataInterface() {propertyValues {Curve {keys [<at 0 1>,<at 1 -0.2>]}}})
    Index 0 Param 2=0.3
    Index 0 Param 3=0.35
    Module.Index0Param0Write=true
    Module.Index0Param1Write=true
    Module.Index0Param2Write=true
    Module.Index0Param3Write=true
   )
  ]
  rendererClasses [&NiagaraMeshRendererProperties]
  renderers [{rendererClass=&NiagaraMeshRendererProperties, rendererIndex=0}]
 }}
summary {
 systemName=NS_Emote_WartyBrine_GlassesSheen
 emitters [{emitterName=BlowingParticles, simTarget=GPUComputeSim, rendererClasses [&NiagaraMeshRendererProperties]}]
 userVariables []
}
