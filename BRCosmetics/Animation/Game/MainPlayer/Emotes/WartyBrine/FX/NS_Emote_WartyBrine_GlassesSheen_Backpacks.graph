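# Niagara effect for the WartyBrine emote, named as the backpack variant of its glasses sheen.
# A GPU mesh emitter follows the skeletal mesh and animates a dynamic material parameter over the particle's lifetime, likely the sheen sweep.
# A delayed burst of two short-lived glint sprites tracks the same mesh, sized and rotated per spawn group and rescaled in the frontend via user variables.
# Both emitters tint by time of day.
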NiagaraSystem /BRCosmetics/Animation/Game/MainPlayer/Emotes/WartyBrine/FX/NS_Emote_WartyBrine_GlassesSheen_Backpacks
dependencies {
 usedDataInterfaces [&NiagaraDataInterfaceCurve, &NiagaraDataInterfaceSkeletalMesh]
 usedDynamicInputs [&RandomRangeFloat]
 usedModules [
  &SystemState
  &self:SystemSpawnScript.SystemScriptSource.SystemScriptGraph.NiagaraNodeAssignment_0.SetVariables_#1
 ]
}
emitters {
 BlowingParticles {
  simTarget=GPUComputeSim
  emitterUpdateScript [
   &EmitterState()
   &SpawnBurst_Instantaneous(Use Spawn Probability=false, UseLoopCountLimit=false)
  ]
  particleSpawnScript [
   &InitializeParticle(
    Color=&FNTimeOfDayColor(
     Night Color=<rgba 0.0815 0.1401 0.8698 1>
     Morning Color=<rgba 0.502 0.3059 0.1843 1>
     Day Color=<rgba 0.5 0.5 0.5 1>
     Evening Color=<rgba 0.502 0.3059 0.1843 1>
    )
    Lifetime=1.2
    Mass Max=1.5
    Mass Mode=::NewEnumerator3
    Mesh Scale Mode=::NewEnumerator3
    Mesh Uniform Scale=0.85
    Position Offset Coordinate Space=::Local
    Sprite Rotation Mode=::NewEnumerator2
    Sprite Size Mode=::NewEnumerator1
    Uniform Sprite Size Max=8
    Uniform Sprite Size Min=4
    UsePositionOffset=true
    Write Material Random=true
   )
   &InitialMeshOrientation(Mesh Axis To Orient=<0 0 1>, Mesh Orientation Mode=::NewEnumerator2, Use Rotation Vector=true)
  ]
  particleUpdateScript [
   &ParticleState()
   &DynamicMaterialParameters(
    Index 0 Param 1=&FloatFromCurve(FloatCurve=DataInterface() {propertyValues {Curve {keys [<at 0 1>,<at 1 -0.2>]}}})
    Index 0 Param 2=0.3
    Index 0 Param 3=0.35
    Module.Index0Param0Write=true
    Module.Index0Param1Write=true
    Module.Index0Param2Write=true
    Module.Index0Param3Write=true
   )
   &SkeletalMeshLocation(
    Bone Sampling Mode=::NewEnumerator0
    MaskSpawnGroup=false
    OverrideRandomSeed=false
    Sampled Position Offset=<7.5 -0.7 -1>
    Skeletal Mesh=$System.Skeletal Mesh
   )
  ]
  rendererClasses [&NiagaraMeshRendererProperties]
  renderers [{rendererClass=&NiagaraMeshRendererProperties, rendererIndex=0}]
 }
 BurstEmbers {
  emitterUpdateScript [
   &EmitterState()
   &SpawnBurst_Instantaneous(Spawn Count=2, Spawn Time=0.84, Use Spawn Probability=false, UseLoopCountLimit=false)
  ]
  particleSpawnScript [
   &FN_SetSpawnGroup(Number of Spawn Groups=2)
   &InitializeParticle(
    Color=&FNTimeOfDayColor(
     Night Color=<rgba 0.2371 0.2808 0.8261 1>
     Morning Color=<rgba 0.502 0.3059 0.1843 1>
     Day Color=<rgba 0.5 0.5 0.5 1>
     Evening Color=<rgba 0.502 0.3059 0.1843 1>
    )
    Lifetime=&Lerp_Float(A=0.3, B=0.3, Alpha=&MakeFloatFromInt(Int=$Particles.SpawnGroup))
    Sprite Rotation Angle=&MaskFloatBySpawnGroup(Masked Float=0.05, Passthrough Float=0.38)
    Sprite Rotation Mode=::NewEnumerator1
    Sprite Size Mode=::NewEnumerator3
    Uniform Sprite Size=&MaskFloatBySpawnGroup(
     Masked Float=&Multiply_Float(
      A=8
      B=&MakeCustomFloatFromBool(Bool=$User.bIsFrontEnd, True Float=$User.FrontendScaler, False Float=1)
     )
     Passthrough Float=&Multiply_Float(
      A=12
      B=&MakeCustomFloatFromBool(Bool=$User.bIsFrontEnd, True Float=$User.FrontendScaler, False Float=1)
     )
    )
    UsePositionOffset=true
    Write Material Random=true
   )
   &CameraOffset(Camera Offset Amount=5)
  ]
  particleUpdateScript [
   &ParticleState()
   &ScaleSpriteSize(Scale Factor=&Vector2DFromFloat(Value=&FloatFromCurve(FloatCurve=DataInterface() {propertyValues {Curve {keys [
          {arriveTangent=11.6538, interpMode=RCIM_Cubic, time=0, value=0.0211}
          {arriveTangent=0.3289, interpMode=RCIM_Cubic, leaveTangent=0.3289, time=0.125, value=1.5}
          <at 1 0.35 RCIM_Cubic>
         ]}}})))
   &ScaleColor(
    Color Value To Scale=$Particles.Color
    Scale Alpha=&FloatFromCurve(FloatCurve=DataInterface() {propertyValues {Curve {keys [<at 0 0 RCIM_Cubic>,<at 0.1 1 RCIM_Cubic>,<at 1 0>]}}})
    Scale RGB=&VectorFromFloat(Value=1)
    ScaleA=false
    ScaleRGB=true
   ) {module=ScaleColor001}
   &SkeletalMeshLocation(
    MaskSpawnGroup=false
    Mesh Sampling Type=::NewEnumerator5
    Sampled Position Offset=<0 -1 -0.5>
    Skeletal Mesh=$System.Skeletal Mesh
    Socket Sampling Mode=::NewEnumerator5
   )
  ]
  rendererClasses [&NiagaraSpriteRendererProperties]
  renderers [{rendererClass=&NiagaraSpriteRendererProperties}]
 }
}
summary {
 systemName=NS_Emote_WartyBrine_GlassesSheen_Backpacks
 emitters [
  {emitterName=BlowingParticles, simTarget=GPUComputeSim, rendererClasses [&NiagaraMeshRendererProperties]}
  {emitterName=BurstEmbers, rendererClasses [&NiagaraSpriteRendererProperties]}
 ]
 userVariables [
  User.bIsFrontEnd {type=NiagaraBool}
  User.FrontendScaler {type=NiagaraFloat, defaultValue=1, description=""}
 ]
}
