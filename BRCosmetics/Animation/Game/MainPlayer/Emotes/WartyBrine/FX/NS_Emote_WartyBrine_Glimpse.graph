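# Niagara sprite effect for the WartyBrine emote, a very brief "glimpse" flash.
# The active BurstEmbers emitter bursts two camera-offset sprites split by spawn group for size and rotation, tinted by time of day, that pop up in scale and then shrink.
# Sprite size switches to a user-supplied scaler in the frontend.
# A second emitter, PixelFlash, with SubUV animation, distance culling and per-particle LOD, is disabled.
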NiagaraSystem /BRCosmetics/Animation/Game/MainPlayer/Emotes/WartyBrine/FX/NS_Emote_WartyBrine_Glimpse
dependencies {
 usedDataInterfaces [&NiagaraDataInterfaceCurve, &NiagaraDataInterfaceColorCurve, &NiagaraDataInterfaceSkeletalMesh]
 usedDynamicInputs []
 usedModules [
  &self:SystemSpawnScript.SystemScriptSource.SystemScriptGraph.NiagaraNodeAssignment_0.SetVariables_#1
  &SystemState
 ]
}
emitters {
 BurstEmbers {
  emitterUpdateScript [
   &EmitterState()
   &SpawnBurst_Instantaneous(Spawn Count=2, Use Spawn Probability=false, UseLoopCountLimit=false)
  ]
  particleSpawnScript [
   &FN_SetSpawnGroup(Number of Spawn Groups=2)
   &InitializeParticle(
    Color=&FNTimeOfDayColor(
     Night Color=<rgba 0.2371 0.2808 0.8261 1>
     Morning Color=<rgba 0.502 0.3059 0.1843 1>
     Day Color=<rgba 0.5 0.5 0.5 1>
     Evening Color=<rgba 0.502 0.3059 0.1843 1>
    )
    Lifetime=&Lerp_Float(A=0.3, B=0.3, Alpha=&MakeFloatFromInt(Int=$Particles.SpawnGroup))
    Sprite Rotation Angle=&MaskFloatBySpawnGroup(Masked Float=0.05, Passthrough Float=0.38)
    Sprite Rotation Mode=::NewEnumerator1
    Sprite Size Mode=::NewEnumerator3
    Uniform Sprite Size=&MaskFloatBySpawnGroup(
     Masked Float=&Multiply_Float(
      A=16.5
      B=&MakeCustomFloatFromBool(Bool=$User.bIsFrontEnd, True Float=$User.FrontendScaler, False Float=1)
     )
     Passthrough Float=&Multiply_Float(
      A=12
      B=&MakeCustomFloatFromBool(Bool=$User.bIsFrontEnd, True Float=$User.FrontendScaler, False Float=1)
     )
    )
    UsePositionOffset=true
    Write Material Random=true
   )
   &CameraOffset(Camera Offset Amount=5)
  ]
  particleUpdateScript [
   &ParticleState()
   &ScaleSpriteSize(Scale Factor=&Vector2DFromFloat(Value=&FloatFromCurve(FloatCurve=DataInterface() {propertyValues {Curve {keys [
          {arriveTangent=11.6538, interpMode=RCIM_Cubic, time=0, value=0.0211}
          {arriveTangent=0.3289, interpMode=RCIM_Cubic, leaveTangent=0.3289, time=0.125, value=1.5}
          <at 1 0.35 RCIM_Cubic>
         ]}}})))
   &ScaleColor(
    Color Value To Scale=$Particles.Color
    Scale Alpha=&FloatFromCurve(FloatCurve=DataInterface() {propertyValues {Curve {keys [<at 0 0 RCIM_Cubic>,<at 0.1 1 RCIM_Cubic>,<at 1 0>]}}})
    Scale RGB=&VectorFromFloat(Value=1)
    ScaleA=false
    ScaleRGB=true
   ) {module=ScaleColor001}
  ]
  rendererClasses [&NiagaraSpriteRendererProperties]
  renderers [{rendererClass=&NiagaraSpriteRendererProperties}]
 }
 PixelFlash {
  bEnabled=false
  emitterUpdateScript [
   &EmitterState(
    Enable Distance Culling=true
    Max Distance Response=::KillImmediately
    Module.bUseMaxDistance=true
    Module.bUseMinDistance=false
    Scalability Mode=::NewEnumerator1
   )
   &SpawnBurst_Instantaneous(Spawn Count=2, Use Spawn Probability=false, UseLoopCountLimit=false)
  ]
  particleSpawnScript [
   &FN_SetSpawnGroup(Number of Spawn Groups=2)
   &InitializeParticle(
    Color=$System.SystemColorFinal
    Lifetime=&Lerp_Float(A=0.4, B=0.25, Alpha=&MakeFloatFromInt(Int=$Particles.SpawnGroup))
    Sprite Size Mode=::NewEnumerator3
    Uniform Sprite Size=&Multiply_Float(
     A=12
     B=&MakeCustomFloatFromBool(Bool=$User.bIsFrontEnd, True Float=$User.FrontendScaler, False Float=1)
    )
    UsePositionOffset=true
    Write Material Random=true
   )
   &FN_PerParticleLOD(Enable SpriteSize Distance Scaling=true, Sprite Size Scale Far=<2.5 2.5>)
   &CameraOffset(Camera Offset Amount=10)
   /SetVariables(Particles.RandomFloat=&UniformRangedFloat(Minimum=0, Maximum=1)) {enabled=false, isSetParameters=true}
   /SetVariables(
    Particles.SubImageIndex=&RandomRangeFloat(Minimum=0, Maximum=3)
    Particles.SubImageIndexAlt:NiagaraFloat=&RandomRangeFloat(Minimum=0, Maximum=3)
   ) {enabled=false, isSetParameters=true}
   &SubUVAnimation(End Frame=3, SubUV Animation Mode:ENiagara_SubUVLookupMode=::NewEnumerator2)
  ]
  particleUpdateScript [
   &ParticleState()
   &ScaleColor(
    Linear Color Curve=DataInterface() {propertyValues {
      AlphaCurve {keys [<at 0.1347 1>,<at 0.8992 0>]}
      BlueCurve {keys [<at 0.1978 1>,<at 0.9037 0.7659>]}
      GreenCurve {keys [<at 0.1978 0.125>,<at 0.9037 0.4597>]}
      RedCurve {keys [<at 0.1978 0.5625>,<at 0.9037 1>]}
     }}
    Scale Mode=::NewEnumerator2
   ) {enabled=false}
   &ScaleColor(
    Linear Color Curve=DataInterface() {propertyValues {
      AlphaCurve {keys [<at 0.3624 1>,<at 0.8992 0>]}
      BlueCurve {keys [<at 0.1978 1.5>]}
      GreenCurve {keys [<at 0.1978 1.5>]}
      RedCurve {keys [<at 0.1978 1.5>]}
     }}
    Scale Mode=::NewEnumerator2
   ) {module=ScaleColor002}
   &ScaleColor(
    Scale RGB=&MakeVectorFromLinearColorRGB(Color=&ColorFromCurve(ColorCurve=DataInterface() {propertyValues {
        AlphaCurve {keys [<at 0 1 RCIM_Cubic>,<at 0.5155 0 RCIM_Cubic>]}
        BlueCurve {keys [<at 0.1913 1 RCIM_Cubic>,<at 0.3986 0.6915 RCIM_Cubic>]}
        GreenCurve {keys [<at 0.1913 0.072 RCIM_Cubic>,<at 0.3986 0.5183 RCIM_Cubic>]}
        RedCurve {keys [<at 0.1913 0.6597 RCIM_Cubic>,<at 0.3986 1 RCIM_Cubic>]}
       }}))
    ScaleA=true
    ScaleRGB=true
   ) {enabled=false, module=ScaleColor001}
   &ScaleColor(
    Scale Alpha=&FNTimeOfDayScalar(Night Scale=2, Morning Scale=1, Evening Scale=2)
    Scale RGB=&VectorFromFloat(Value=&FNTimeOfDayScalar(Night Scale=4, Morning Scale=1, Evening Scale=4))
    ScaleA=true
    ScaleRGB=true
   ) {module=ScaleColor003}
   &ScaleSpriteSize(
    Uniform Curve Scale=1.1
    Uniform Curve Sprite Scale=DataInterface() {propertyValues {Curve {keys [{arriveTangent=2.6565, interpMode=RCIM_Cubic, leaveTangent=2.6565, tangentMode=RCTM_User, time=0, value=0}, <at 1 1 RCIM_Cubic>]}}}
   ) {module=ScaleSpriteSize001}
  ]
  rendererClasses [&NiagaraSpriteRendererProperties]
  renderers [{rendererClass=&NiagaraSpriteRendererProperties}]
 }
}
summary {
 systemName=NS_Emote_WartyBrine_Glimpse
 emitters [
  {bEnabled=false, emitterName=PixelFlash, rendererClasses [&NiagaraSpriteRendererProperties]}
  {emitterName=BurstEmbers, rendererClasses [&NiagaraSpriteRendererProperties]}
 ]
 userVariables [
  User.bIsFrontEnd {type=NiagaraBool}
  User.FrontendScaler {type=NiagaraFloat, defaultValue=1, description=""}
  User.IsAlt {type=NiagaraFloat, defaultValue=0, description=""}
 ]
}
